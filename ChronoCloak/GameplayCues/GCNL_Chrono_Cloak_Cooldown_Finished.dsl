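; GameplayCue Notify Looping for the Chrono Cloak cooldown-finished cue, gating its end-of-relevance handling on a tag.
; OnCeaseRelevant checks whether the target actor has the NightmareEquipped gameplay tag.
; With the tag, it runs and returns the parent cue's behavior; otherwise it returns false and skips it.
; The construction script is empty.
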(fn OnCeaseRelevant
  (params (MyTarget :type Actor%20Object%20Reference) (Parameters :type Gameplay%20Cue%20Parameters%20Structure%20%28by%20ref%29))
  (bind _returnvalue (GameplayTags|HasMatchingGameplayTag(Message) MyTarget (Variables|Default|GetTag_NightmareEquipped)))
  (if _returnvalue
    (bind _returnvalue_2 (|CallParentFunction MyTarget Parameters))
    (return _returnvalue_2)
    (else
      (return false))))

(fn ConstructionScript)
